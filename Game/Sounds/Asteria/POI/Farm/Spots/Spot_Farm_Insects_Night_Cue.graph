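# Sound cue for night insect spot ambience at the Farm POI, using Apollo farm biome recordings.
# A Random node picks among seven waves from two insect sets, with the last InsectNightB variant weighted double.
# A Modulator at the root varies pitch and volume around normal for each play.
# It plays at reduced volume through a one-shot tile ambient attenuation and the outdoor ambient sound class.
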SoundCue /Game/Sounds/Asteria/POI/Farm/Spots/Spot_Farm_Insects_Night_Cue
root=Modulator_0
nodes [
 {id=Modulator_0, children [Random_0], properties {pitchMax=1.2, pitchMin=0.8, volumeMax=1.2, volumeMin=0.8}}
 {
  id=Random_0
  children [WavePlayer_0,WavePlayer_1,WavePlayer_2,WavePlayer_3,WavePlayer_6,WavePlayer_8,WavePlayer_9]
  properties {weights [1,1,1,1,1,1,2]}
 }
 {id=WavePlayer_0, properties {soundWaveAssetPtr=/Game/Sounds/Apollo/Biomes/Farm/Biome_Farm_InsectNight_01}}
 {id=WavePlayer_1, properties {soundWaveAssetPtr=/Game/Sounds/Apollo/Biomes/Farm/Biome_Farm_InsectNight_02}}
 {id=WavePlayer_2, properties {soundWaveAssetPtr=/Game/Sounds/Apollo/Biomes/Farm/Biome_Farm_InsectNight_03}}
 {id=WavePlayer_3, properties {soundWaveAssetPtr=/Game/Sounds/Apollo/Biomes/Farm/Biome_Farm_InsectNight_04}}
 {id=WavePlayer_6, properties {soundWaveAssetPtr=/Game/Sounds/Apollo/Biomes/Farm/Biome_Farm_InsectNightB_01}}
 {id=WavePlayer_8, properties {soundWaveAssetPtr=/Game/Sounds/Apollo/Biomes/Farm/Biome_Farm_InsectNightB_02}}
 {id=WavePlayer_9, properties {soundWaveAssetPtr=/Game/Sounds/Apollo/Biomes/Farm/Biome_Farm_InsectNightB_03}}
]
properties {
 attenuationSettings=/Game/Sounds/Attenuation/Fort_TileAmbient_OneShot_Att
 maxDistance=8000
 soundClassObject=&Fort_Ambient_Outdoors_Athena
 volumeMultiplier=0.451
 preEffectBusSends [{}]
}
