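# Sound cue for the looping hover tractor spot sound at the Asteria farm point of interest.
# A random node picks one of three looping hover tractor waves with equal weight, then a modulator slightly varies pitch and volume.
# The cue plays quietly at lowered pitch and overrides attenuation with its own shape and falloff distance.
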SoundCue /Game/Sounds/Asteria/POI/Farm/Spots/Spot_Hover_Tractor_Cue
root=Modulator_0
nodes [
 {id=Modulator_0, children [Random_1], properties {pitchMax=1.2, pitchMin=0.8, volumeMax=1.1, volumeMin=0.9}}
 {id=Random_1, children [WavePlayer_10, WavePlayer_11, WavePlayer_12], properties {weights [1,1,1]}}
 {id=WavePlayer_10, properties {bLooping=true, soundWaveAssetPtr=&/Spot_Hover_Tractor_Loop_03}}
 {id=WavePlayer_11, properties {bLooping=true, soundWaveAssetPtr=&/Spot_Hover_Tractor_Loop_01}}
 {id=WavePlayer_12, properties {bLooping=true, soundWaveAssetPtr=&/Spot_Hover_Tractor_Loop_02}}
]
properties {
 bOverrideAttenuation=true
 maxDistance=1470
 pitchMultiplier=0.6
 volumeMultiplier=0.2
 attenuationOverrides {falloffDistance=1250, attenuationShapeExtents {x=220}}
}
